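; Looping gameplay cue for the PistolCanary single-tentacle energy blast, managing a Niagara effect tied to the target player pawn's body mesh.
; On application, it caches the first spawned particle component as Niagara and feeds the pawn's Body skeletal mesh into its "SK" parameter.
; On removal, it deactivates that cached component and releases it to the pool.
; Both events print yellow debug strings.
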(event EventOnApplication(EitherVFX-preferthisone)
  (params (MyTarget :type Actor%20Object%20Reference) (Parameters :type Gameplay%20Cue%20Parameters%20Structure%20%28by%20ref%29) (ParticleComponents :type Array%20of%20FXSystem%20Component%20Object%20References) (AudioComponents :type Array%20of%20Audio%20Component%20Object%20References) (BurstCameraShakeInstance :type Camera%20Shake%20Base%20Object%20Reference) (BurstDecalInstance :type Decal%20Actor%20Object%20Reference))
  (bind _asfort_player_pawn (Utilities|Casting|CastToFortPlayerPawn MyTarget))
  (bind _asniagara_particle_system_component (Utilities|Casting|CastToNiagaraComponent (Utilities|Array|Get(acopy) ParticleComponents)))
  (bind _output_get (Variables|Default|SetAs%20Niagara%20Particle%20System%20Component _asniagara_particle_system_component))
  (Niagara|SetNiagaraSkeletalMeshComponent _output_get "SK" (Customization|GetSkeletalMeshforPartType _asfort_player_pawn "Body" :!class /Script/FortniteGame.FortPlayerPawn))
  (Development|PrintString 0 "BGA :: EFFECT ADDED" false true "(R=1.000000,G=0.913380,B=0.000000,A=1.000000)"))

(event EventOnRemoval(EitherVFX-preferthisversion)
  (params (MyTarget :type Actor%20Object%20Reference) (Parameters :type Gameplay%20Cue%20Parameters%20Structure%20%28by%20ref%29) (ParticleComponents :type Array%20of%20FXSystem%20Component%20Object%20References) (AudioComponents :type Array%20of%20Audio%20Component%20Object%20References) (BurstCameraShakeInstance :type Camera%20Shake%20Base%20Object%20Reference) (BurstDecalInstance :type Decal%20Actor%20Object%20Reference))
  (bind _as_niagara_particle_system_component (Variables|Default|GetAs%20Niagara%20Particle%20System%20Component))
  (Components|Activation|Deactivate _as_niagara_particle_system_component)
  (Effects|Components|ParticleSystem|ReleasetoPool _as_niagara_particle_system_component)
  (Development|PrintString 0 "BGA:: EFFECT REMOVED" false true "(R=1.000000,G=0.913380,B=0.000000,A=1.000000)"))

(fn ConstructionScript)
